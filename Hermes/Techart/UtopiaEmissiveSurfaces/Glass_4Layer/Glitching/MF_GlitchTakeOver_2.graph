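# Material function for Utopia emissive glass that outputs a glitch frequency mask, a square color glitch mask and a randomized time.
# Actor position offsets a time-driven sine, which is smoothstepped into bursts, multiplied by random flicker and inverted.
# A position hash offsets time, which is floored to pan a grid texture in steps for the square glitch.
# The texture sample takes explicit UV derivatives for mip selection.
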MaterialFunction /Hermes/Techart/UtopiaEmissiveSurfaces/Glass_4Layer/Glitching/MF_GlitchTakeOver_2
graph {nodes [
  n1 = MaterialFunctionCall() {materialFunction=/Game/MF__FloatFlickerRandom_Generic}
  Divide(A=@n1.RandomFlickerFloat) {constB=20}
  n2 = Constant() {r=1}
  n3 = FunctionInput(Preview=Constant() {r=0.075}) {bUsePreviewValueAsDefault=true, inputName=GlitchDuration, inputType=FunctionInput_Scalar}
  n4 = MaterialFunctionCall(Float3=Divide(A=ActorPositionWS()) {constB=4000}) {materialFunction=&BreakOutFloat3Components}
  n5 = Add(A=Add(A=@n4.R, B=OneMinus(None=@n4.G)), B=Time())
  n6 = FunctionInput(Preview=Constant() {r=0.1}) {bUsePreviewValueAsDefault=true, inputName=GlitchFrequency, inputType=FunctionInput_Scalar}
  n7 = Multiply(A=Sine(None=Multiply(A=@n5, B=@n6) {constB=0.2})) {constB=0.5}
  n8 = FunctionInput(Preview=Constant() {r=2}) {bUsePreviewValueAsDefault=true, inputName=Glitch Flickering, inputType=FunctionInput_Scalar}
  n9 = Multiply(
   A=SmoothStep(Min=Subtract(A=@n2, B=@n3), Max=@n2, Value=Add(A=@n7) {constB=0.5}) {constMax=0.9, constMin=0.8}
   B=Saturate(None=Multiply(A=@n1.RandomFlickerFloat, B=@n8) {constB=5})
  )
  FunctionOutput(None=NamedRerouteDeclaration(None=OneMinus(None=@n9)) {name=GlitchFrequencyMask, nodeColor=<rgba 0.9377 1 0.8953 1>}) {outputName=GlitchFrequencyMask}
  n10 = NamedRerouteUsage()
  n11 = FunctionInput(Preview=Constant() {r=10}) {bUsePreviewValueAsDefault=true, inputName=Glitch Size, inputType=FunctionInput_Scalar, sortPriority=4}
  n12 = Constant() {r=2}
  n13 = NamedRerouteUsage()
  n14 = ComponentMask(None=TransformPosition(None=Constant3Vector()) {transformType=TRANSFORMPOSSOURCE_World}) {g=true, r=true}
  n15 = Ceil(None=AppendVector(A=Add(A=@n13, B=@n14), B=Constant()))
  n16 = Constant() {desc="LeapOfFaith ", r=1000}
  n17 = MaterialFunctionCall(Float3=Multiply(A=Divide(A=@n15, B=@n16)) {constB=11.17}) {materialFunction=&BreakOutFloat3Components}
  n18 = Frac(None=Add(A=@n17.R, B=@n17.R))
  n19 = NamedRerouteDeclaration(None=Add(A=Multiply(A=@n18) {constB=120}, B=Time())) {name=Random time, nodeColor=<rgba 0 1 0.7422 1>}
  n20 = Divide(A=Divide(
    A=@n19
    B=FunctionInput(Preview=@n12) {bUsePreviewValueAsDefault=true, inputName=Color Glitch Speed, inputType=FunctionInput_Scalar, sortPriority=4}
   )) {constB=2.37}
  n21 = TextureObject() {samplerType=SAMPLERTYPE_LinearColor, texture=/Game/Effects/Fort_Effects/Effects/Characters/Generic/ResIn/DiffusedStylizedGridAnimOffsets}
  n22 = TextureCoordinate()
  n23 = TextureSample!(
   UVs=Panner(Coordinate=Divide(A=@n10, B=@n11) {constB=10}, Time=Floor(None=@n20)) {speedX=0.3, speedY=0.1}
   Tex=FunctionInput(Preview=@n21) {bUsePreviewValueAsDefault=true, inputName=Texture Glitch, inputType=FunctionInput_Texture2D, sortPriority=5}
   "DDX(UVs)"=FunctionInput(Preview=DDX(Value=@n22)) {inputName=DDX, inputType=FunctionInput_Vector2, sortPriority=2}
   "DDY(UVs)"=FunctionInput(Preview=DDY(Value=@n22)) {inputName=DDY, inputType=FunctionInput_Vector2, sortPriority=3}
   Apply View MipBias
  ) {
   mipValueMode=TMVM_Derivative
   samplerType=SAMPLERTYPE_LinearColor
   texture=&DiffusedStylizedGridAnimOffsets
  }
  n24 = OneMinus(None=Multiply(A=NamedRerouteUsage(), B=OneMinus(None=@n23.R)))
  FunctionOutput(None=NamedRerouteDeclaration(None=Saturate(None=@n24)) {name=Sqaure Color glitch, nodeColor=<rgba 0.2422 1 0 1>}) {outputName=Sqaure Color glitch}
  FunctionOutput(None=@n19) {outputName=Random time}
  Multiply(A=@n10) {constB=0.1}
  NamedRerouteDeclaration(None=FunctionInput(Preview=TextureCoordinate()) {inputName=Triplanar UV, inputType=FunctionInput_Vector2, sortPriority=4}) {name=Trip UV, nodeColor=<rgba 0 1 0.8359 1>}
  NamedRerouteUsage()
  ScalarParameter() {defaultValue=0.1, group=HolographicTexture_Glitch, parameterName=GlitchFrequency, sortPriority=10}
  ScalarParameter() {defaultValue=0.075, group=HolographicTexture_Glitch, parameterName=GlitchDuration, sortPriority=10}
 ]}
